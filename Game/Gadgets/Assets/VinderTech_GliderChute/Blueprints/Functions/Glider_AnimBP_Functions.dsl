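; Animation helper function for the VinderTech GliderChute that eases a float toward a forward, back or idle value based on local forward acceleration.
; Strong positive or negative acceleration selects the forward or back target, while the near-zero check and the fallback both select idle.
; Each target has its own interpolation speed, and FInterpTo returns the new value from the current value and the passed-in time.
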(fn ForwardBack_Value_Interp
  (params (Time :type Float%20%28double-precision%29) (Currnet_Value :type Float%20%28double-precision%29) (Idle_Value :type Float%20%28double-precision%29 :default 1.0) (Idle_InterpSpeed :type Float%20%28double-precision%29 :default 1.0) (Forward_Value :type Float%20%28double-precision%29 :default 2.0) (Forward_InterpSpeed :type Float%20%28double-precision%29 :default 2.0) (Back_Value :type Float%20%28double-precision%29 :default -2.0) (Back_InterpSpeed :type Float%20%28double-precision%29 :default 2.0) (LocalAccelForward :type Float%20%28double-precision%29))
  (if (> LocalAccelForward 5e+02)
    (return (Math|Interpolation|FInterpTo Currnet_Value Forward_Value Time Forward_InterpSpeed))
    (elif (Math|Float|NearlyEqual(Float) LocalAccelForward 0.0 1e+01)
      (return (Math|Interpolation|FInterpTo Currnet_Value Idle_Value Time Idle_InterpSpeed))
      (elif (< LocalAccelForward -5e+02)
        (return (Math|Interpolation|FInterpTo Currnet_Value Back_Value Time Back_InterpSpeed))
        (else
          (return (Math|Interpolation|FInterpTo Currnet_Value Idle_Value Time Idle_InterpSpeed)))))))
