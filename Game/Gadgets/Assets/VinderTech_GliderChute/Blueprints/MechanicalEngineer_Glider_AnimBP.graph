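# Animation Blueprint for the Mechanical Engineer owl glider, animating deployment, gliding and wing flaps.
# A Locomotion state machine drives closed, deploy, idle and flap states, layered with lean, turn, feather, blink, head-dart and caw additives.
# Matching backpacks blend in a pose snapshot and world-space Frame transform; cosmetic preview pins feet to wrench sockets.
# The event graph randomizes flap timing and style and accesses the owl pet's AnimBP.
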AnimBlueprint /Game/Gadgets/Assets/VinderTech_GliderChute/Blueprints/MechanicalEngineer_Glider_AnimBP
target_skeleton=/Game/Gadgets/Assets/VinderTech_GliderChute/Glider_Mechanical_Engineer_Female/Meshes/Mechanical_Engineer_Female_Glider_Skeleton
graphs [
 AnimGraph {nodes [
   AnimGraphNode_StateMachine_1 = CuhCAW() {is=title AnimGraphNode_StateMachine}
   Apply Additive!(Base, Additive=@AnimGraphNode_StateMachine_1, bAlphaBoolEnabled:bool=Get bCanCaw() -> (bCanCaw:bool)) -> (Pose) {alphaInputType=Bool, alphaBoolBlend {blendInTime=0.1, blendOption=Linear, blendOutTime=0.1}}
   n1 = "Transform (Modify) Bone - Bone: Frame"(
    ComponentPose=Local To Component(LocalPose=Use cached pose 'Backpack'())
    Alpha=Get BackpackCurveValue() -> (BackpackCurveValue:float)
    Translation=Get BackpackOwlSnapshotLocation() -> (BackpackOwlSnapshotLocation:Vector)
    Rotation=Get BackpackOwlSnapshotRotation() -> (BackpackOwlSnapshotRotation:Rotator)
    Scale="1,1,1"
   ) {
    rotationMode=BMM_Replace
    rotationSpace=BCS_WorldSpace
    translationMode=BMM_Replace
    translationSpace=BCS_WorldSpace
    alphaBoolBlend {blendOption=Linear}
    boneToModify {boneName=Frame}
   }
   n2 = Blend Poses by bool(
    BlendPose_0=Component To Local(ComponentPose=@n1)
    BlendPose_1=Use cached pose 'Locomotion'()
    bActiveValue=Get bBackpacksMatch() -> (bBackpacksMatch:bool)
   ) {blendType=Linear}
   n3 = Use cached pose 'FrontEnd_Feet'()
   n4 = Get bIsCosmeticPreview() -> (bIsCosmeticPreview:bool)
   Output Pose(Result=Slot 'DefaultSlot'(Source=Blend Poses by bool(BlendPose_0=@n3, BlendPose_1=@n2, bActiveValue=@n4) {blendType=Linear})) {layerGroup=None}
   n5 = Pose Snapshot!(Snapshot:Pose Snapshot Structure=Get BackpackSnapshotPose() -> (BackpackSnapshotPose:Pose Snapshot Structure)) -> (Pose:Pose Link Structure) {is=title AnimGraphNode_PoseSnapshot, mode=SnapshotPin}
   n6 = Use cached pose 'Locomotion'()
   n7 = Get BackPackPoseCurveValue() -> (BackPackPoseCurveValue:float)
   Save cached pose 'Backpack'(Pose=Slot 'Owl'(Source=Layered blend per bone(BasePose=@n6, BlendPoses_0=@n5, BlendWeights_0=@n7) {bBlendRootMotionBasedOnRootBone=false, layerSetup [{branchFilters [{boneName=pelvis}]}]}) {slotName=Owl}) {cacheName=Backpack}
   n8 = Break Fort Anim Input Skydiving!(FortAnimInput_Skydiving:"Fort Anim Input Skydiving Structure (by ref)"=Get SkydivingState() -> (SkydivingState:Fort Anim Input Skydiving Structure)) -> (
    ParaGilderRootOffset:Vector=~
    ParaGliderRootRotationOverride:Rotator=~
    bUseParaGlideRootModifier:bool=~
    bIsSkydivingFromLaunchPad:bool=~
    bIsSkydivingFromBus:bool=~
    bIsInVortex:bool=~
    bIsDiving:bool=~
    bIsParachuteOpen:bool=~
    bIsSkydiving:bool=~
    bIsParachuteLeaning:bool=~
    bIsSkydiveLeaning:bool=~
    bIsLeaning:bool=~
    bIsSkydiveDiveMode:bool=~
    bPlaySkydiveDrift:bool=~
    bSkydiveDriftDelayActive:bool=~
    bSkydiveDriftAnimAllowed:bool=~
    bIsGliderRight:bool=~
    bIsGliderCenter:bool=~
    bIsGliderLeft:bool=~
    bIsGliderForward:bool=~
    bIsGliderBack:bool=~
    bHasDeployedAndIsStandingUpright:bool=~
    LocalAccelForward:float=~
    LocalAccelRight:float=~
    LocalVelocityRight:float=~
    SkydiveAimPitch:float=~
    SkydiveAimYaw:float=~
    DeployChuteAnimRate:float=~
    SkydiveDriftAnimRate:float=~
    SkydiveDriftAnimRateCurrent:float=~
    SkydiveFidgetAnimRate:float=~
    SkydiveFidgetAnimRateCurrent:float=~
    SkydiveAdditiveAlpha:float=~
    SkydiveDriftDelay:float=~
    SkydiveDriftAnim:int
    SkydiveDriftAnimMax:int
    LaunchpadAnim:int
    LocalAccelDir:ESkydivingDirection Enum
    DirectionLast:ESkydivingDirection Enum
    PoseAdditive:Anim Sequence Object Reference
   ) {is=title K2Node_BreakStruct}
   AnimGraphNode_StateMachine_0 = Locomotion() {is=title AnimGraphNode_StateMachine}
   n9 = Break Fort Anim Input Skydiving(FortAnimInput_Skydiving=Get SkydivingState()) -> (
    ParaGilderRootOffset=~
    ParaGliderRootRotationOverride=~
    bUseParaGlideRootModifier=~
    bIsSkydivingFromLaunchPad=~
    bIsSkydivingFromBus=~
    bIsInVortex=~
    bIsDiving=~
    bIsParachuteOpen=~
    bIsSkydiving=~
    bIsParachuteLeaning=~
    bIsSkydiveLeaning=~
    bIsLeaning=~
    bIsSkydiveDiveMode=~
    bPlaySkydiveDrift=~
    bSkydiveDriftDelayActive=~
    bSkydiveDriftAnimAllowed=~
    bIsGliderRight=~
    bIsGliderCenter=~
    bIsGliderLeft=~
    bIsGliderForward=~
    bIsGliderBack=~
    bHasDeployedAndIsStandingUpright=~
    LocalAccelForward=~
    LocalAccelRight=~
    LocalVelocityRight=~
    SkydiveAimPitch=~
    SkydiveAimYaw=~
    DeployChuteAnimRate=~
    SkydiveDriftAnimRate=~
    SkydiveDriftAnimRateCurrent=~
    SkydiveFidgetAnimRate=~
    SkydiveFidgetAnimRateCurrent=~
    SkydiveAdditiveAlpha=~
    SkydiveDriftDelay=~
   )
   n10 = Apply Additive!(
    Base=Apply Additive!(
     Base=@AnimGraphNode_StateMachine_0
     Additive=Blendspace Player 'BS_MechanicalEngineer_Lean_GLIDER'(X=@n8.LocalAccelRight, Y=@n8.LocalAccelForward) {blendSpace=/Game/Animation/Game/MainPlayer/Skydive/ParaGlide/MechanicalEngineer/BS_MechanicalEngineer_Lean_GLIDER}
     AlphaCurveName:Name=ParaGlide_Lean
    ) -> (Pose) {alphaInputType=Curve, alphaBoolBlend {blendOption=Linear}, alphaScaleBiasClamp {bias=1, scale=-1}}
    Additive=Blendspace Player 'BS_MechanicalEngineer_Turn_GLIDER'(X=@n9.SkydiveAimYaw, Y=@n9.SkydiveAimPitch) {blendSpace=/Game/Animation/Game/MainPlayer/Skydive/ParaGlide/MechanicalEngineer/BS_MechanicalEngineer_Turn_GLIDER}
    Alpha:float
   ) -> (Pose) {alphaBoolBlend {blendOption=Linear}}
   n11 = Random Sequence Player!() -> (Pose:Pose Link Structure) {
    is=title AnimGraphNode_RandomPlayer
    bShuffleMode=true
    entries [
     {
      chanceToPlay=1
      maxLoopCount=0
      maxPlayRate=1.35
      minLoopCount=0
      minPlayRate=1
      sequence=/Game/Animation/Game/MainPlayer/Skydive/ParaGlide/MechanicalEngineer/MechanicalEngineer_WingSway_GLIDER
      blendIn {blendOption=Linear, blendTime=0.2, customCurve=null}
     }
     {
      chanceToPlay=1
      maxLoopCount=0
      maxPlayRate=1.35
      minLoopCount=0
      minPlayRate=1
      sequence=/Game/Animation/Game/MainPlayer/Skydive/ParaGlide/MechanicalEngineer/MechanicalEngineer_WingSway_B_GLIDER
      blendIn {blendOption=Linear, blendTime=0.2, customCurve=null}
     }
     {
      chanceToPlay=1
      maxLoopCount=0
      maxPlayRate=1.45
      minLoopCount=0
      minPlayRate=1.25
      sequence=&MechanicalEngineer_WingSway_B_GLIDER
      blendIn {blendOption=Linear, blendTime=0.2, customCurve=null}
     }
    ]
   }
   n12 = Get WingPoseAdditiveAlpha() -> (WingPoseAdditiveAlpha:float)
   n13 = Get FeatherSpeed() -> (FeatherSpeed:float)
   n14 = Random Sequence Player() {
    bShuffleMode=true
    entries [
     {
      chanceToPlay=1
      maxLoopCount=1
      maxPlayRate=1
      minLoopCount=0
      minPlayRate=1
      sequence=/Game/Animation/Game/MainPlayer/Skydive/ParaGlide/MechanicalEngineer/MechanicalEngineer_Head_Darts_Left_GLIDER
      blendIn {blendOption=Linear, blendTime=0.2, customCurve=null}
     }
     {
      chanceToPlay=1
      maxLoopCount=1
      maxPlayRate=1
      minLoopCount=0
      minPlayRate=1
      sequence=/Game/Animation/Game/MainPlayer/Skydive/ParaGlide/MechanicalEngineer/MechanicalEngineer_Head_Darts_Right_GLIDER
      blendIn {blendOption=Linear, blendTime=0.2, customCurve=null}
     }
    ]
   }
   n15 = Apply Additive!(
    Base=Apply Additive!(
     Base=Apply Additive(Base=@n10, Additive=@n11, Alpha=@n12) {alphaBoolBlend {blendOption=Linear}}
     Additive="Sequence Player (additive) 'MechanicalEngineer_Feathers_GLIDER'"!(PlayRate:float=@n13) -> (Pose) {sequence=/Game/Animation/Game/MainPlayer/Skydive/ParaGlide/MechanicalEngineer/MechanicalEngineer_Feathers_GLIDER}
     Alpha:float
    ) -> (Pose) {alphaBoolBlend {blendOption=Linear}}
    Additive=@n14
    bAlphaBoolEnabled:bool=Get bCanLook() -> (bCanLook:bool)
   ) -> (Pose) {alphaInputType=Bool, alphaBoolBlend {blendInTime=0.1, blendOption=Linear, blendOutTime=0.1}}
   n16 = Random Sequence Player() {
    bShuffleMode=true
    entries [
     {
      chanceToPlay=1
      maxLoopCount=1
      maxPlayRate=1
      minLoopCount=0
      minPlayRate=1
      sequence=/Game/Animation/Game/MainPlayer/Skydive/ParaGlide/MechanicalEngineer/MechanicalEngineer_Blink_GLIDER
      blendIn {blendOption=Linear, blendTime=0.2, customCurve=null}
     }
     {
      chanceToPlay=1
      maxLoopCount=1
      maxPlayRate=1
      minLoopCount=0
      minPlayRate=1
      sequence=/Game/Animation/Game/MainPlayer/Skydive/ParaGlide/MechanicalEngineer/MechanicalEngineer_BlinkDart_GLIDER
      blendIn {blendOption=Linear, blendTime=0.2, customCurve=null}
     }
     {
      chanceToPlay=1
      maxLoopCount=0
      maxPlayRate=1
      minLoopCount=0
      minPlayRate=1
      sequence=/Game/Animation/Game/MainPlayer/Skydive/ParaGlide/MechanicalEngineer/MechanicalEngineer_BlinkDart_Right_GLIDER
      blendIn {blendOption=Linear, blendTime=0.2, customCurve=null}
     }
    ]
   }
   n17 = Get bCanBlink() -> (bCanBlink:bool)
   Save cached pose 'Locomotion'(Pose=Apply Additive(
     Base=Apply Additive!(Base=@n15, Additive=@n16, bAlphaBoolEnabled:bool=@n17) -> (Pose) {alphaInputType=Bool, alphaBoolBlend {blendOption=Linear}}
     Additive="Sequence Player (additive) 'MechanicalEngineer_Glide_Pose_Add_GLIDER'"() {sequence=/Game/Animation/Game/MainPlayer/Skydive/ParaGlide/MechanicalEngineer/MechanicalEngineer_Glide_Pose_Add_GLIDER}
     Alpha=Get WingPoseAdditiveAlpha()
    ) {alphaBoolBlend {blendInTime=0.1, blendOption=Linear, blendOutTime=0.1}}) {cacheName=Locomotion}
   n18 = "Transform (Modify) Bone - Bone: foot_r"!(
    ComponentPose=Local To Component(LocalPose=Use cached pose 'Locomotion'())
    AlphaCurveName:Name=IK_Feet
    Translation=Get Right_Foot_Loc() -> (Right_Foot_Loc:Vector)
    Rotation=Get Right_Foot_Rot() -> (Right_Foot_Rot:Rotator)
    Scale="1,1,1"
   ) -> (Pose) {
    alphaInputType=Curve
    rotationMode=BMM_Replace
    rotationSpace=BCS_WorldSpace
    translationMode=BMM_Replace
    translationSpace=BCS_WorldSpace
    alphaBoolBlend {blendOption=Linear}
    alphaScaleBiasClamp {bMapRange=true, bias=1, scale=-1, outRange {max=0, min=1}}
    boneToModify {boneName=foot_r}
   }
   n19 = "Transform (Modify) Bone - Bone: foot_l"!(
    ComponentPose=@n18
    AlphaCurveName:Name=IK_Feet
    Translation=Get Left_Foot_loc() -> (Left_Foot_loc:Vector)
    Rotation=Get Left_Foot_Rot() -> (Left_Foot_Rot:Rotator)
    Scale="1,1,1"
   ) -> (Pose) {
    alphaInputType=Curve
    rotationMode=BMM_Replace
    rotationSpace=BCS_WorldSpace
    translationMode=BMM_Replace
    translationSpace=BCS_WorldSpace
    alphaBoolBlend {blendOption=Linear}
    alphaScaleBiasClamp {bMapRange=true, bias=1, scale=-1, outRange {max=0, min=1}}
    boneToModify {boneName=foot_l}
   }
   Save cached pose 'FrontEnd_Feet'(Pose=Component To Local(ComponentPose=@n19)) {cacheName=FrontEnd_Feet}
   "Sequence Player (additive) 'MechanicalEngineer_WingSway_GLIDER'"() {sequence=&MechanicalEngineer_WingSway_GLIDER}
   AND Boolean(A=@n17, B=NOT Boolean(A=Get bCanCaw()))
   Get IsCosmeticPreview!(self:Fort Player Parachute Object Reference=Get OwnerGlider() -> (OwnerGlider:Fort Player Parachute Object Reference)) -> (IsCosmeticPreview:bool)
  ]}
 Locomotion {
  entry=Closed
  path=AnimGraph.AnimGraphNode_StateMachine_0.Locomotion
  ids=id
  nodes [
   AnimStateNode_0 = Idle_Open() {is=title AnimStateNode}
   AnimStateNode_1 = Open_Into() {is=title AnimStateNode, bAlwaysResetOnEntry=true, stateEntered {guid=#1, notifyName=DeployEntered}}
   AnimStateNode_2 = Closed() {is=title AnimStateNode, stateEntered {guid=#2, notifyName=ClosedEntered}}
   AnimStateNode_3 = Flap_Transition() {is=title AnimStateNode}
   AnimStateNode_4 = Flap() {is=title AnimStateNode}
   AnimStateNode_5 = Open_Into_GeyserVent() {is=title AnimStateNode, stateEntered {guid=#3, notifyName=DeployEntered}}
   AnimStateNode_6 = Open_Into_GliderItem() {is=title AnimStateNode, stateEntered {guid=#3, notifyName=DeployEntered}}
   AnimStateTransitionNode_0 = AnimStateTransitionNode() {bAutomaticRuleBasedOnSequencePlayerInState=true, blendMode=Linear, crossfadeDuration=0}
   AnimStateTransitionNode_1 = AnimStateTransitionNode() {crossfadeDuration=0}
   AnimStateTransitionNode_10 = AnimStateTransitionNode() {crossfadeDuration=0}
   AnimStateTransitionNode_11 = AnimStateTransitionNode() {crossfadeDuration=0}
   AnimStateTransitionNode_12 = AnimStateTransitionNode() {crossfadeDuration=0}
   AnimStateTransitionNode_13 = AnimStateTransitionNode() {bAutomaticRuleBasedOnSequencePlayerInState=true, crossfadeDuration=0}
   AnimStateTransitionNode_14 = AnimStateTransitionNode()
   AnimStateTransitionNode_15 = AnimStateTransitionNode() {crossfadeDuration=0}
   AnimStateTransitionNode_16 = AnimStateTransitionNode() {crossfadeDuration=0}
   AnimStateTransitionNode_17 = AnimStateTransitionNode() {bAutomaticRuleBasedOnSequencePlayerInState=true, crossfadeDuration=0}
   AnimStateTransitionNode_18 = AnimStateTransitionNode()
   AnimStateTransitionNode_2 = AnimStateTransitionNode() {crossfadeDuration=0}
   AnimStateTransitionNode_3 = AnimStateTransitionNode() {crossfadeDuration=0, priorityOrder=2}
   AnimStateTransitionNode_4 = AnimStateTransitionNode()
   AnimStateTransitionNode_5 = AnimStateTransitionNode() {bAutomaticRuleBasedOnSequencePlayerInState=true, blendMode=Linear, crossfadeDuration=0.35}
   AnimStateTransitionNode_6 = AnimStateTransitionNode() {priorityOrder=2}
   AnimStateTransitionNode_7 = AnimStateTransitionNode() {crossfadeDuration=0}
   AnimStateTransitionNode_8 = AnimStateTransitionNode()
   AnimStateTransitionNode_9 = AnimStateTransitionNode() {bAutomaticRuleBasedOnSequencePlayerInState=true, blendMode=Linear, crossfadeDuration=0.35}
  ]
  links [
   {
    dst=Open_Into
    rule=AnimGraph.AnimGraphNode_StateMachine_0.Locomotion.AnimStateTransitionNode_1.Transition
    src=Closed
    via=AnimStateTransitionNode_1
   }
   {
    dst=Open_Into_GeyserVent
    rule=AnimGraph.AnimGraphNode_StateMachine_0.Locomotion.AnimStateTransitionNode_16.Transition
    src=Closed
    via=AnimStateTransitionNode_16
   }
   {
    dst=Open_Into_GliderItem
    rule=AnimGraph.AnimGraphNode_StateMachine_0.Locomotion.AnimStateTransitionNode_11.Transition
    src=Closed
    via=AnimStateTransitionNode_11
   }
   {
    dst=Closed
    rule=AnimGraph.AnimGraphNode_StateMachine_0.Locomotion.AnimStateTransitionNode_7.Transition
    src=Flap
    via=AnimStateTransitionNode_7
   }
   {
    dst=Idle_Open
    rule=AnimGraph.AnimGraphNode_StateMachine_0.Locomotion.AnimStateTransitionNode_5.Transition
    src=Flap
    via=AnimStateTransitionNode_5
   }
   {
    dst=Closed
    rule=AnimGraph.AnimGraphNode_StateMachine_0.Locomotion.AnimStateTransitionNode_10.Transition
    src=Flap_Transition
    via=AnimStateTransitionNode_10
   }
   {
    dst=Idle_Open
    rule=AnimGraph.AnimGraphNode_StateMachine_0.Locomotion.AnimStateTransitionNode_9.Transition
    src=Flap_Transition
    via=AnimStateTransitionNode_9
   }
   {
    dst=Closed
    rule=AnimGraph.AnimGraphNode_StateMachine_0.Locomotion.AnimStateTransitionNode_2.Transition
    src=Idle_Open
    via=AnimStateTransitionNode_2
   }
   {
    dst=Flap
    rule=AnimGraph.AnimGraphNode_StateMachine_0.Locomotion.AnimStateTransitionNode_4.Transition
    src=Idle_Open
    via=AnimStateTransitionNode_4
   }
   {
    dst=Flap_Transition
    rule=AnimGraph.AnimGraphNode_StateMachine_0.Locomotion.AnimStateTransitionNode_8.Transition
    src=Idle_Open
    via=AnimStateTransitionNode_8
   }
   {
    dst=Closed
    rule=AnimGraph.AnimGraphNode_StateMachine_0.Locomotion.AnimStateTransitionNode_3.Transition
    src=Open_Into
    via=AnimStateTransitionNode_3
   }
   {
    dst=Flap
    rule=AnimGraph.AnimGraphNode_StateMachine_0.Locomotion.AnimStateTransitionNode_6.Transition
    src=Open_Into
    via=AnimStateTransitionNode_6
   }
   {
    dst=Idle_Open
    rule=AnimGraph.AnimGraphNode_StateMachine_0.Locomotion.AnimStateTransitionNode_0.Transition
    src=Open_Into
    via=AnimStateTransitionNode_0
   }
   {
    dst=Closed
    rule=AnimGraph.AnimGraphNode_StateMachine_0.Locomotion.AnimStateTransitionNode_15.Transition
    src=Open_Into_GeyserVent
    via=AnimStateTransitionNode_15
   }
   {
    dst=Flap
    rule=AnimGraph.AnimGraphNode_StateMachine_0.Locomotion.AnimStateTransitionNode_18.Transition
    src=Open_Into_GeyserVent
    via=AnimStateTransitionNode_18
   }
   {
    dst=Idle_Open
    rule=AnimGraph.AnimGraphNode_StateMachine_0.Locomotion.AnimStateTransitionNode_17.Transition
    src=Open_Into_GeyserVent
    via=AnimStateTransitionNode_17
   }
   {
    dst=Closed
    rule=AnimGraph.AnimGraphNode_StateMachine_0.Locomotion.AnimStateTransitionNode_12.Transition
    src=Open_Into_GliderItem
    via=AnimStateTransitionNode_12
   }
   {
    dst=Flap
    rule=AnimGraph.AnimGraphNode_StateMachine_0.Locomotion.AnimStateTransitionNode_14.Transition
    src=Open_Into_GliderItem
    via=AnimStateTransitionNode_14
   }
   {
    dst=Idle_Open
    rule=AnimGraph.AnimGraphNode_StateMachine_0.Locomotion.AnimStateTransitionNode_13.Transition
    src=Open_Into_GliderItem
    via=AnimStateTransitionNode_13
   }
  ]
 }
 Idle_Open {
  path=AnimGraph.AnimGraphNode_StateMachine_0.Locomotion.AnimStateNode_0.Idle_Open
  nodes [
   Blend Poses by bool(bActiveValue=Get IsCosmeticPreview(self=Get OwnerGlider())) {blendType=Linear}
   n1 = Break Fort Anim Input Skydiving(FortAnimInput_Skydiving=Get SkydivingState()) -> (
    ParaGilderRootOffset=~
    ParaGliderRootRotationOverride=~
    bUseParaGlideRootModifier=~
    bIsSkydivingFromLaunchPad=~
    bIsSkydivingFromBus=~
    bIsInVortex=~
    bIsDiving=~
    bIsParachuteOpen=~
    bIsSkydiving=~
    bIsParachuteLeaning=~
    bIsSkydiveLeaning=~
    bIsLeaning=~
    bIsSkydiveDiveMode=~
    bPlaySkydiveDrift=~
    bSkydiveDriftDelayActive=~
    bSkydiveDriftAnimAllowed=~
    bIsGliderRight=~
    bIsGliderCenter=~
    bIsGliderLeft=~
    bIsGliderForward=~
    bIsGliderBack=~
    bHasDeployedAndIsStandingUpright=~
    LocalAccelForward=~
    LocalAccelRight=~
    LocalVelocityRight=~
    SkydiveAimPitch=~
    SkydiveAimYaw=~
    DeployChuteAnimRate=~
    SkydiveDriftAnimRate=~
    SkydiveDriftAnimRateCurrent=~
    SkydiveFidgetAnimRate=~
    SkydiveFidgetAnimRateCurrent=~
    SkydiveAdditiveAlpha=~
    SkydiveDriftDelay=~
   )
   n2 = Blendspace Player 'BS_MechanicalEngineer_GLIDER'(X=@n1.SkydiveAimYaw, Y=@n1.LocalAccelForward) {
    blendSpace=/Game/Animation/Game/MainPlayer/Skydive/ParaGlide/MechanicalEngineer/BS_MechanicalEngineer_GLIDER
    groupName=Glide
    method=SyncGroup
   }
   n3 = "Sequence Player (additive) 'MechanicalEngineer_Glide_FrontEnd_GLIDER'"() {
    groupName=FrontEndGlide
    method=SyncGroup
    sequence=/Game/Animation/Game/MainPlayer/Skydive/ParaGlide/MechanicalEngineer/MechanicalEngineer_Glide_FrontEnd_GLIDER
   }
   Output Animation Pose(Result=Apply Additive(Base=@n2, Additive=@n3, bAlphaBoolEnabled=Get bIsCosmeticPreview()) {alphaInputType=Bool, alphaBoolBlend {blendOption=Linear}})
  ]
 }
 Open_Into {
  path=AnimGraph.AnimGraphNode_StateMachine_0.Locomotion.AnimStateNode_1.Open_Into
  nodes [
   n1 = Break Fort Anim Input Skydiving(FortAnimInput_Skydiving=Get SkydivingState()) -> (
    ParaGilderRootOffset=~
    ParaGliderRootRotationOverride=~
    bUseParaGlideRootModifier=~
    bIsSkydivingFromLaunchPad=~
    bIsSkydivingFromBus=~
    bIsInVortex=~
    bIsDiving=~
    bIsParachuteOpen=~
    bIsSkydiving=~
    bIsParachuteLeaning=~
    bIsSkydiveLeaning=~
    bIsLeaning=~
    bIsSkydiveDiveMode=~
    bPlaySkydiveDrift=~
    bSkydiveDriftDelayActive=~
    bSkydiveDriftAnimAllowed=~
    bIsGliderRight=~
    bIsGliderCenter=~
    bIsGliderLeft=~
    bIsGliderForward=~
    bIsGliderBack=~
    bHasDeployedAndIsStandingUpright=~
    LocalAccelForward=~
    LocalAccelRight=~
    LocalVelocityRight=~
    SkydiveAimPitch=~
    SkydiveAimYaw=~
    DeployChuteAnimRate=~
    SkydiveDriftAnimRate=~
    SkydiveDriftAnimRateCurrent=~
    SkydiveFidgetAnimRate=~
    SkydiveFidgetAnimRateCurrent=~
    SkydiveAdditiveAlpha=~
    SkydiveDriftDelay=~
   )
   n2 = Blendspace Player 'BS_MechanicalEngineer_Into_NoPack_GLIDER'(X=@n1.SkydiveAimYaw, Y=@n1.SkydiveAimPitch) {bLoop=false, blendSpace=/Game/Animation/Game/MainPlayer/Skydive/ParaGlide/MechanicalEngineer/BS_MechanicalEngineer_Into_NoPack_GLIDER}
   n3 = Blendspace Player 'BS_MechanicalEngineer_Into_GLIDER'(X=@n1.SkydiveAimYaw, Y=@n1.SkydiveAimPitch) {bLoop=false, blendSpace=/Game/Animation/Game/MainPlayer/Skydive/ParaGlide/MechanicalEngineer/BS_MechanicalEngineer_Into_GLIDER}
   n4 = Get bBackpacksMatch()
   n5 = Sequence Player 'MechanicalEngineer_Into_FrontEnd_GLIDER'() {bLoopAnimation=false, sequence=/Game/Animation/Game/MainPlayer/Skydive/ParaGlide/MechanicalEngineer/MechanicalEngineer_Into_FrontEnd_GLIDER}
   Output Animation Pose(Result=Blend Poses by bool(
     BlendPose_0=@n5
     BlendPose_1=Blend Poses by bool(BlendPose_0=@n3, BlendPose_1=@n2, BlendTime_0=0, BlendTime_1=0, bActiveValue=@n4) {blendType=Linear, childUpateMode=ResetChildOnActivate}
     BlendTime_0=0
     BlendTime_1=0
     bActiveValue=Get bIsCosmeticPreview()
    ) {blendType=Linear, childUpateMode=ResetChildOnActivate})
  ]
 }
 Closed {
  path=AnimGraph.AnimGraphNode_StateMachine_0.Locomotion.AnimStateNode_2.Closed
  nodes [
   n1 = Blend Poses by bool(
    BlendPose_0=Sequence Player 'MechanicalEngineer_Closed_FromPack_GLIDER'() {sequence=/Game/Animation/Game/MainPlayer/Skydive/ParaGlide/MechanicalEngineer/MechanicalEngineer_Closed_FromPack_GLIDER}
    BlendPose_1=Sequence Player 'MechanicalEngineer_Closed_GLIDER'() {sequence=/Game/Animation/Game/MainPlayer/Skydive/ParaGlide/MechanicalEngineer/MechanicalEngineer_Closed_GLIDER}
    BlendTime_1=0
    bActiveValue=Get bUseClosedFromPack() -> (bUseClosedFromPack:bool)
   ) {blendType=Linear}
   n2 = Sequence Player 'MechanicalEngineer_Closed_NoPack_GLIDER'() {sequence=/Game/Animation/Game/MainPlayer/Skydive/ParaGlide/MechanicalEngineer/MechanicalEngineer_Closed_NoPack_GLIDER}
   n3 = Get bBackpacksMatch()
   Output Animation Pose(Result=Blend Poses by bool(
     BlendPose_0=Sequence Player 'MechanicalEngineer_Closed_GLIDER'() {sequence=&MechanicalEngineer_Closed_GLIDER}
     BlendPose_1=Blend Poses by bool(BlendPose_0=@n1, BlendPose_1=@n2, BlendTime_0=0, BlendTime_1=0, bActiveValue=@n3) {blendType=Linear}
     BlendTime_0=0
     BlendTime_1=0
     bActiveValue=Get bIsCosmeticPreview()
    ) {blendType=Linear})
  ]
 }
 Flap_Transition {
  path=AnimGraph.AnimGraphNode_StateMachine_0.Locomotion.AnimStateNode_3.Flap_Transition
  nodes [
   n1 = Break Fort Anim Input Skydiving(FortAnimInput_Skydiving=Get SkydivingState()) -> (
    ParaGilderRootOffset=~
    ParaGliderRootRotationOverride=~
    bUseParaGlideRootModifier=~
    bIsSkydivingFromLaunchPad=~
    bIsSkydivingFromBus=~
    bIsInVortex=~
    bIsDiving=~
    bIsParachuteOpen=~
    bIsSkydiving=~
    bIsParachuteLeaning=~
    bIsSkydiveLeaning=~
    bIsLeaning=~
    bIsSkydiveDiveMode=~
    bPlaySkydiveDrift=~
    bSkydiveDriftDelayActive=~
    bSkydiveDriftAnimAllowed=~
    bIsGliderRight=~
    bIsGliderCenter=~
    bIsGliderLeft=~
    bIsGliderForward=~
    bIsGliderBack=~
    bHasDeployedAndIsStandingUpright=~
    LocalAccelForward=~
    LocalAccelRight=~
    LocalVelocityRight=~
    SkydiveAimPitch=~
    SkydiveAimYaw=~
    DeployChuteAnimRate=~
    SkydiveDriftAnimRate=~
    SkydiveDriftAnimRateCurrent=~
    SkydiveFidgetAnimRate=~
    SkydiveFidgetAnimRateCurrent=~
    SkydiveAdditiveAlpha=~
    SkydiveDriftDelay=~
   )
   Output Animation Pose(Result=Blendspace Player 'BS_MechanicalEngineer_Flap_GLIDER'!(X=@n1.SkydiveAimYaw, Y=@n1.LocalAccelForward, PlayRate:float=Get FlapPlayRate() -> (FlapPlayRate:float)) -> (Pose) {blendSpace=/Game/Animation/Game/MainPlayer/Skydive/ParaGlide/MechanicalEngineer/BS_MechanicalEngineer_Flap_GLIDER})
  ]
 }
 Flap {
  path=AnimGraph.AnimGraphNode_StateMachine_0.Locomotion.AnimStateNode_4.Flap
  nodes [
   n1 = Break Fort Anim Input Skydiving(FortAnimInput_Skydiving=Get SkydivingState()) -> (
    ParaGilderRootOffset=~
    ParaGliderRootRotationOverride=~
    bUseParaGlideRootModifier=~
    bIsSkydivingFromLaunchPad=~
    bIsSkydivingFromBus=~
    bIsInVortex=~
    bIsDiving=~
    bIsParachuteOpen=~
    bIsSkydiving=~
    bIsParachuteLeaning=~
    bIsSkydiveLeaning=~
    bIsLeaning=~
    bIsSkydiveDiveMode=~
    bPlaySkydiveDrift=~
    bSkydiveDriftDelayActive=~
    bSkydiveDriftAnimAllowed=~
    bIsGliderRight=~
    bIsGliderCenter=~
    bIsGliderLeft=~
    bIsGliderForward=~
    bIsGliderBack=~
    bHasDeployedAndIsStandingUpright=~
    LocalAccelForward=~
    LocalAccelRight=~
    LocalVelocityRight=~
    SkydiveAimPitch=~
    SkydiveAimYaw=~
    DeployChuteAnimRate=~
    SkydiveDriftAnimRate=~
    SkydiveDriftAnimRateCurrent=~
    SkydiveFidgetAnimRate=~
    SkydiveFidgetAnimRateCurrent=~
    SkydiveAdditiveAlpha=~
    SkydiveDriftDelay=~
   )
   n2 = Get FlapPlayRate()
   n3 = Get FlapBS() -> (FlapBS:Blend Space Object Reference)
   n4 = Blend Poses by bool(
    BlendPose_0=Layered blend per bone(
     BasePose=Blendspace Player!(X=@n1.SkydiveAimYaw, Y=@n1.LocalAccelForward, PlayRate:float=@n2, BlendSpace:Blend Space Object Reference=@n3) -> (Pose)
     BlendPoses_0=Local Space Ref Pose()
    ) {layerSetup [{branchFilters [{boneName=foot_r}, {boneName=foot_l}]}]}
    BlendPose_1=Blendspace Player(X=@n1.SkydiveAimYaw, Y=@n1.LocalAccelForward, PlayRate=Get FlapPlayRate(), BlendSpace=Get FlapBS())
    bActiveValue=Get bIsCosmeticPreview()
   ) {blendType=Linear}
   n5 = "Sequence Player (additive) 'MechanicalEngineer_Glide_FrontEnd_GLIDER'"() {
    groupName=FrontEndGlide
    method=SyncGroup
    sequence=&MechanicalEngineer_Glide_FrontEnd_GLIDER
   }
   n6 = Get bIsCosmeticPreview()
   Output Animation Pose(Result=Apply Additive(
     Base=Apply Additive(Base=@n4, Additive=@n5, bAlphaBoolEnabled=@n6) {alphaInputType=Bool, alphaBoolBlend {blendOption=Linear}}
     Additive="Sequence Player (additive) 'MechanicalEngineer_ClawGrip_Pose_GLIDER'"() {sequence=/Game/Animation/Game/MainPlayer/Skydive/ParaGlide/MechanicalEngineer/MechanicalEngineer_ClawGrip_Pose_GLIDER}
     bAlphaBoolEnabled=Get bIsCosmeticPreview()
    ) {alphaInputType=Bool, alphaBoolBlend {blendInTime=0.05, blendOption=Linear, blendOutTime=0.05}})
  ]
 }
 Open_Into_GeyserVent {
  path=AnimGraph.AnimGraphNode_StateMachine_0.Locomotion.AnimStateNode_5.Open_Into_GeyserVent
  nodes [
   n1 = Break Fort Anim Input Skydiving(FortAnimInput_Skydiving=Get SkydivingState()) -> (
    ParaGilderRootOffset=~
    ParaGliderRootRotationOverride=~
    bUseParaGlideRootModifier=~
    bIsSkydivingFromLaunchPad=~
    bIsSkydivingFromBus=~
    bIsInVortex=~
    bIsDiving=~
    bIsParachuteOpen=~
    bIsSkydiving=~
    bIsParachuteLeaning=~
    bIsSkydiveLeaning=~
    bIsLeaning=~
    bIsSkydiveDiveMode=~
    bPlaySkydiveDrift=~
    bSkydiveDriftDelayActive=~
    bSkydiveDriftAnimAllowed=~
    bIsGliderRight=~
    bIsGliderCenter=~
    bIsGliderLeft=~
    bIsGliderForward=~
    bIsGliderBack=~
    bHasDeployedAndIsStandingUpright=~
    LocalAccelForward=~
    LocalAccelRight=~
    LocalVelocityRight=~
    SkydiveAimPitch=~
    SkydiveAimYaw=~
    DeployChuteAnimRate=~
    SkydiveDriftAnimRate=~
    SkydiveDriftAnimRateCurrent=~
    SkydiveFidgetAnimRate=~
    SkydiveFidgetAnimRateCurrent=~
    SkydiveAdditiveAlpha=~
    SkydiveDriftDelay=~
   )
   n2 = Blendspace Player 'BS_MechanicalEngineer_Into_NoPack_GLIDER'(X=@n1.SkydiveAimYaw, Y=@n1.SkydiveAimPitch) {bLoop=false, blendSpace=&BS_MechanicalEngineer_Into_NoPack_GLIDER}
   Output Animation Pose(Result=Blend Poses by bool(
     BlendPose_0=Blendspace Player 'BS_MechanicalEngineer_Into_GLIDER'(X=@n1.SkydiveAimYaw, Y=@n1.SkydiveAimPitch) {bLoop=false, blendSpace=&BS_MechanicalEngineer_Into_GLIDER}
     BlendPose_1=@n2
     BlendTime_0=0
     BlendTime_1=0
     bActiveValue=Get bBackpacksMatch()
    ) {blendType=Linear})
  ]
 }
 Open_Into_GliderItem {
  path=AnimGraph.AnimGraphNode_StateMachine_0.Locomotion.AnimStateNode_6.Open_Into_GliderItem
  nodes [
   n1 = Break Fort Anim Input Skydiving(FortAnimInput_Skydiving=Get SkydivingState()) -> (
    ParaGilderRootOffset=~
    ParaGliderRootRotationOverride=~
    bUseParaGlideRootModifier=~
    bIsSkydivingFromLaunchPad=~
    bIsSkydivingFromBus=~
    bIsInVortex=~
    bIsDiving=~
    bIsParachuteOpen=~
    bIsSkydiving=~
    bIsParachuteLeaning=~
    bIsSkydiveLeaning=~
    bIsLeaning=~
    bIsSkydiveDiveMode=~
    bPlaySkydiveDrift=~
    bSkydiveDriftDelayActive=~
    bSkydiveDriftAnimAllowed=~
    bIsGliderRight=~
    bIsGliderCenter=~
    bIsGliderLeft=~
    bIsGliderForward=~
    bIsGliderBack=~
    bHasDeployedAndIsStandingUpright=~
    LocalAccelForward=~
    LocalAccelRight=~
    LocalVelocityRight=~
    SkydiveAimPitch=~
    SkydiveAimYaw=~
    DeployChuteAnimRate=~
    SkydiveDriftAnimRate=~
    SkydiveDriftAnimRateCurrent=~
    SkydiveFidgetAnimRate=~
    SkydiveFidgetAnimRateCurrent=~
    SkydiveAdditiveAlpha=~
    SkydiveDriftDelay=~
   )
   n2 = Blendspace Player 'BS_MechanicalEngineer_Into_NoPack_GLIDER'(X=@n1.SkydiveAimYaw, Y=@n1.SkydiveAimPitch) {bLoop=false, blendSpace=&BS_MechanicalEngineer_Into_NoPack_GLIDER}
   n3 = Blendspace Player 'BS_MechanicalEngineer_Into_FromPack_GLIDER'(X=@n1.SkydiveAimYaw, Y=@n1.SkydiveAimPitch) {bLoop=false, blendSpace=/Game/Animation/Game/MainPlayer/Skydive/ParaGlide/MechanicalEngineer/BS_MechanicalEngineer_Into_FromPack_GLIDER}
   Output Animation Pose(Result=Blend Poses by bool(BlendPose_0=@n3, BlendPose_1=@n2, BlendTime_0=0, BlendTime_1=0, bActiveValue=Get bBackpacksMatch()) {blendType=Linear, childUpateMode=ResetChildOnActivate})
  ]
 }
 Transition {
  path=AnimGraph.AnimGraphNode_StateMachine_0.Locomotion.AnimStateTransitionNode_0.Transition
  nodes [
   Result()
   "Equal (Float)"!(A:double="Get Relevant Anim Time Remaining (Open_Into)"!() -> (ReturnValue:float) {is=title K2Node_AnimGetter}, B:double=0) -> (ReturnValue:bool) {is=title K2Node_CallFunction}
  ]
 }
 Transition {
  path=AnimGraph.AnimGraphNode_StateMachine_0.Locomotion.AnimStateTransitionNode_1.Transition
  nodes [
   Result(bCanEnterTransition=Get bTransition_Closed_To_OpenInto() -> (bTransition_Closed_To_OpenInto:bool))
   Get BlockAnimRule() -> (BlockAnimRule:bool)
  ]
 }
 Transition {
  path=AnimGraph.AnimGraphNode_StateMachine_0.Locomotion.AnimStateTransitionNode_10.Transition
  nodes [Result(bCanEnterTransition=Get bTransition_To_Closed() -> (bTransition_To_Closed:bool))]
 }
 Transition {
  path=AnimGraph.AnimGraphNode_StateMachine_0.Locomotion.AnimStateTransitionNode_11.Transition
  nodes [Result(bCanEnterTransition=Get bTransition_Closed_To_OpenIntoGliderItem() -> (bTransition_Closed_To_OpenIntoGliderItem:bool))]
 }
 Transition {
  path=AnimGraph.AnimGraphNode_StateMachine_0.Locomotion.AnimStateTransitionNode_12.Transition
  nodes [Result(bCanEnterTransition=Get bTransition_To_Closed())]
 }
 Transition {path=AnimGraph.AnimGraphNode_StateMachine_0.Locomotion.AnimStateTransitionNode_13.Transition, nodes [Result()]}
 Transition {
  path=AnimGraph.AnimGraphNode_StateMachine_0.Locomotion.AnimStateTransitionNode_14.Transition
  nodes [
   Result(bCanEnterTransition=Get BlockAnimRule())
   n1 = GetCurveValue(CurveName=DisableFlap)
   AND Boolean(
    A=OR Boolean(A=Get bCanFlap() -> (bCanFlap:bool), B=Get bCanFlapTransition() -> (bCanFlapTransition:bool))
    B=NOT Boolean(A="Equal (Float)"(A=@n1, B=1))
   )
  ]
 }
 Transition {
  path=AnimGraph.AnimGraphNode_StateMachine_0.Locomotion.AnimStateTransitionNode_15.Transition
  nodes [Result(bCanEnterTransition=Get bTransition_To_Closed())]
 }
 Transition {
  path=AnimGraph.AnimGraphNode_StateMachine_0.Locomotion.AnimStateTransitionNode_16.Transition
  nodes [Result(bCanEnterTransition=Get bTransition_Closed_To_OpenIntoGeyserVent() -> (bTransition_Closed_To_OpenIntoGeyserVent:bool))]
 }
 Transition {path=AnimGraph.AnimGraphNode_StateMachine_0.Locomotion.AnimStateTransitionNode_17.Transition, nodes [Result()]}
 Transition {
  path=AnimGraph.AnimGraphNode_StateMachine_0.Locomotion.AnimStateTransitionNode_18.Transition
  nodes [
   Result(bCanEnterTransition=Get BlockAnimRule())
   n1 = GetCurveValue(CurveName=DisableFlap)
   AND Boolean(A=OR Boolean(A=Get bCanFlap(), B=Get bCanFlapTransition()), B=NOT Boolean(A="Equal (Float)"(A=@n1, B=1)))
  ]
 }
 Transition {
  path=AnimGraph.AnimGraphNode_StateMachine_0.Locomotion.AnimStateTransitionNode_2.Transition
  nodes [Result(bCanEnterTransition=Get bTransition_To_Closed())]
 }
 Transition {
  path=AnimGraph.AnimGraphNode_StateMachine_0.Locomotion.AnimStateTransitionNode_3.Transition
  nodes [Result(bCanEnterTransition=Get bTransition_To_Closed())]
 }
 Transition {
  path=AnimGraph.AnimGraphNode_StateMachine_0.Locomotion.AnimStateTransitionNode_4.Transition
  nodes [Result(bCanEnterTransition=Get bTransition_IdleOpen_To_Flap() -> (bTransition_IdleOpen_To_Flap:bool))]
 }
 Transition {
  path=AnimGraph.AnimGraphNode_StateMachine_0.Locomotion.AnimStateTransitionNode_5.Transition
  nodes [
   Result()
   NOT Boolean(A=Get bCanFlap())
  ]
 }
 Transition {
  path=AnimGraph.AnimGraphNode_StateMachine_0.Locomotion.AnimStateTransitionNode_6.Transition
  nodes [Result(bCanEnterTransition=Get bTransition_OpenInto_To_Flap() -> (bTransition_OpenInto_To_Flap:bool))]
 }
 Transition {
  path=AnimGraph.AnimGraphNode_StateMachine_0.Locomotion.AnimStateTransitionNode_7.Transition
  nodes [Result(bCanEnterTransition=Get bTransition_To_Closed())]
 }
 Transition {
  path=AnimGraph.AnimGraphNode_StateMachine_0.Locomotion.AnimStateTransitionNode_8.Transition
  nodes [
   Result(bCanEnterTransition=Get bCanFlapTransition())
   AND Boolean!(A=NOT Boolean(A=Get bIsFlapping() -> (bIsFlapping:bool)), B=Get bIsAccelerating2D(), C:bool) -> (ReturnValue)
   Get BlockAnimRule()
  ]
 }
 Transition {
  path=AnimGraph.AnimGraphNode_StateMachine_0.Locomotion.AnimStateTransitionNode_9.Transition
  nodes [
   Result()
   OR Boolean(A=NOT Boolean(A=Get bCanFlap()), B=NOT Boolean(A=Get bCanFlapTransition()))
  ]
 }
 CuhCAW {entry=Caw, path=AnimGraph.AnimGraphNode_StateMachine_1.CuhCAW, ids=id, nodes [AnimStateNode_0 = Caw() {is=title AnimStateNode}]}
 Caw {
  path=AnimGraph.AnimGraphNode_StateMachine_1.CuhCAW.AnimStateNode_0.Caw
  nodes [Output Animation Pose(Result="Sequence Player (additive) 'MechanicalEngineer_Caw_GLIDER'"() {sequence=/Game/Animation/Game/MainPlayer/Skydive/ParaGlide/MechanicalEngineer/MechanicalEngineer_Caw_GLIDER})]
 }
 EventGraph {nodes [
   Event BlueprintUpdateAnimation()
   > Branch(Condition=Get bIsCosmeticPreview())
   n1 = Get ParachuteMesh!(self:Fort Player Parachute Object Reference=Get OwnerGlider()) -> (ParachuteMesh:Skeletal Mesh Component Budgeted Object Reference)
   n2 = BreakTransform(InTransform=GetSocketTransform(self=@n1, InSocketName=wrench_right_foot, TransformSpace=RTS_World))
   > Set Right_Foot_Loc(Right_Foot_Loc:Vector=@n2) -> (Output_Get:Vector)
   n3 = BreakTransform(InTransform=GetSocketTransform(self=@n1, InSocketName=wrench_left_foot, TransformSpace=RTS_World))
   > Set Left_Foot_loc(Left_Foot_loc:Vector=@n3) -> (Output_Get:Vector)
   > Set Right_Foot_Rot(Right_Foot_Rot:Rotator=@n2.Rotation) -> (Output_Get:Rotator)
   > Set Left_Foot_Rot(Left_Foot_Rot:Rotator=@n3.Rotation) -> (Output_Get:Rotator)
   AnimNotify_TurnONEngines() {is=title K2Node_Event}
   > SetGliderFullyDeployed!(execute, self:Fort Glider Anim Instance Object Reference, bIsFullyDeployed:bool=true) -> (then)
   n4 = AnimNotify_ResetFlap() {is=title K2Node_Event}
   n5 = Sequence(execute=[@n4.then, @n18.then])
   > Set bCanFlapTransition(bCanFlapTransition:bool=false) -> (Output_Get:bool)
   Set bCanFlap(execute=@n5.then_1, bCanFlap:bool=false) -> (Output_Get:bool)
   n6 = Get SkydivingState()
   > n7 = Branch(Condition="float < float"!(
     A:double=Break Fort Anim Input Skydiving(FortAnimInput_Skydiving=@n6) -> (
      ParaGilderRootOffset=~
      ParaGliderRootRotationOverride=~
      bUseParaGlideRootModifier=~
      bIsSkydivingFromLaunchPad=~
      bIsSkydivingFromBus=~
      bIsInVortex=~
      bIsDiving=~
      bIsParachuteOpen=~
      bIsSkydiving=~
      bIsParachuteLeaning=~
      bIsSkydiveLeaning=~
      bIsLeaning=~
      bIsSkydiveDiveMode=~
      bPlaySkydiveDrift=~
      bSkydiveDriftDelayActive=~
      bSkydiveDriftAnimAllowed=~
      bIsGliderRight=~
      bIsGliderCenter=~
      bIsGliderLeft=~
      bIsGliderForward=~
      bIsGliderBack=~
      bHasDeployedAndIsStandingUpright=~
      LocalAccelForward=~
      LocalAccelRight=~
      LocalVelocityRight=~
      SkydiveAimPitch=~
      SkydiveAimYaw=~
      DeployChuteAnimRate=~
      SkydiveDriftAnimRate=~
      SkydiveDriftAnimRateCurrent=~
      SkydiveFidgetAnimRate=~
      SkydiveFidgetAnimRateCurrent=~
      SkydiveAdditiveAlpha=~
      SkydiveDriftDelay=~
     ).LocalAccelForward
     B:double=-250
    ) -> (ReturnValue:bool) {is=title K2Node_CallFunction})
   > n8 = Delay(Duration=0)
   Set FlapPlayRate(
    execute=[@n13.then, @n8.then, @n19.Completed]
    FlapPlayRate:float=Select(Option 0:double=1, Option 1:double=1.1, Index:int=RandomInteger!(Max:int=2) -> (ReturnValue:int=0)) -> (ReturnValue:double)
   ) -> (Output_Get:float)
   > Set FlapBS(FlapBS:Blend Space Object Reference=Select(
     Option 0:Blend Space Object Reference=&BS_MechanicalEngineer_Flap_GLIDER
     Option 1:Blend Space Object Reference=/Game/Animation/Game/MainPlayer/Skydive/ParaGlide/MechanicalEngineer/BS_MechanicalEngineer_Flap2x_GLIDER
     Index=RandomInteger(Max=2) -> (ReturnValue=0)
    ) -> (ReturnValue:Blend Space Object Reference)) -> (Output_Get:Blend Space Object Reference)
   n9 = Get SkydivingState()
   > n10 = Branch(Condition="float < float"(
     A=Break Fort Anim Input Skydiving(FortAnimInput_Skydiving=@n9) -> (
      ParaGilderRootOffset=~
      ParaGliderRootRotationOverride=~
      bUseParaGlideRootModifier=~
      bIsSkydivingFromLaunchPad=~
      bIsSkydivingFromBus=~
      bIsInVortex=~
      bIsDiving=~
      bIsParachuteOpen=~
      bIsSkydiving=~
      bIsParachuteLeaning=~
      bIsSkydiveLeaning=~
      bIsLeaning=~
      bIsSkydiveDiveMode=~
      bPlaySkydiveDrift=~
      bSkydiveDriftDelayActive=~
      bSkydiveDriftAnimAllowed=~
      bIsGliderRight=~
      bIsGliderCenter=~
      bIsGliderLeft=~
      bIsGliderForward=~
      bIsGliderBack=~
      bHasDeployedAndIsStandingUpright=~
      LocalAccelForward=~
      LocalAccelRight=~
      LocalVelocityRight=~
      SkydiveAimPitch=~
      SkydiveAimYaw=~
      DeployChuteAnimRate=~
      SkydiveDriftAnimRate=~
      SkydiveDriftAnimRateCurrent=~
      SkydiveFidgetAnimRate=~
      SkydiveFidgetAnimRateCurrent=~
      SkydiveAdditiveAlpha=~
      SkydiveDriftDelay=~
     ).LocalAccelForward
     B=-250
    ))
   > n11 = Set FlapPlayRate(FlapPlayRate=1.1)
   Set bCanFlap(execute=[@n10.else, @n11.then], bCanFlap=true)
   > Set bCanFlapTransition(bCanFlapTransition=false)
   n12 = Set Flap_Delay(execute=@n7.else, Flap_Delay:double=RandomFloatInRange(Min=3, Max=5)) -> (Output_Get:double)
   > n13 = Delay(Duration=@n12)
   AnimNotify_ResetFreefall() {is=title K2Node_Event}
   > Set bSkydiveFreefall(bSkydiveFreefall:bool=false) -> (Output_Get:bool)
   > Branch(Condition=Get bBackpacksMatch())
   n14 = Get CosmeticPetInstance!(self:Fort Player Pawn Object Reference=GetFortPlayerPawn!(self:Fort Player Parachute Object Reference=Get OwnerGlider()) -> (ReturnValue:Fort Player Pawn Object Reference)) -> (CosmeticPetInstance:Fort Player Pet Object Reference)
   > n15 = Cast To B_MechanicalOwl_Pet_AnimBP!(
    execute
    Object=GetAnimInstance(self=Get PetMesh!(self:Fort Player Pet Object Reference=@n14) -> (PetMesh:Skeletal Mesh Component Budgeted Object Reference))
   ) -> (then, CastFailed, AsB Mechanical Owl Pet Anim BP:B Mechanical Owl Pet Anim BP Object Reference) {is=title K2Node_DynamicCast}
   > n16 = Cast To MechanicalOwlAnimLayer!(
    execute
    Object=GetLinkedAnimLayerInstanceByGroup!(self:Anim Instance Object Reference=@n15, InGroup:Name=PetOverride) -> (ReturnValue:Anim Instance Object Reference)
   ) -> (then, CastFailed, AsMechanical Owl Anim Layer:Mechanical Owl Anim Layer Object Reference) {is=title K2Node_DynamicCast}
   > Set bSkydiveFreefall!(execute, bSkydiveFreefall=false, self:Layer Anim Instance Mechanical Engineer Pet Object Reference=@n16) -> (then, Output_Get)
   Event BlueprintInitializeAnimation()
   > n17 = Set bCanBlink(bCanBlink:bool=true) -> (Output_Get:bool)
   Set bCanLook(execute=[@n20.then, @n17.then], bCanLook:bool=true) -> (Output_Get:bool)
   > Delay(Duration=0.4)
   > Set bCanCaw(bCanCaw:bool=Select(Option 0:bool=false, Option 1:bool=true, Index=RandomInteger(Max=2) -> (ReturnValue=0)) -> (ReturnValue:bool)) -> (Output_Get:bool)
   n18 = AnimNotify_FlapOff() {is=title K2Node_Event}
   AnimNotify_FlapEnable() {is=title K2Node_Event}
   > n19 = Do Once()
   AnimNotify_LookOff() {is=title K2Node_Event}
   > Set bCanLook(bCanLook=false)
   > n20 = Delay(Duration=4)
   AnimNotify_DeployEntered() {is=title K2Node_Event}
   > Set DeployAnimAlpha(DeployAnimAlpha:float=1) -> (Output_Get:float)
   > Delay(Duration=0.2)
   > Set bHideBackpackOwl(bHideBackpackOwl:bool=false) -> (Output_Get:bool)
   n21 = Get OwnerGlider()
   > Set Actor Hidden In Game!(execute, self:Actor Object Reference=@n21, bNewHidden:bool=false) -> (then) {is=title K2Node_CallFunction}
   AnimNotify_CawOff() {is=title K2Node_Event}
   > Set bCanCaw(bCanCaw=false)
   > Delay(Duration=Select(Option 0:double=4, Option 1:double=8, Index=RandomInteger(Max=2) -> (ReturnValue=0)) -> (ReturnValue:double))
   > n22 = Branch(Condition=Get bCanLook())
   Set bCanCaw(execute=@n22.else, bCanCaw=true)
   AnimNotify_BlinkOff() {is=title K2Node_Event}
   > Set bCanBlink(bCanBlink=false)
   > Delay(Duration=RandomFloatInRange(Min=1, Max=3))
   > Set bCanBlink(bCanBlink=true)
   AnimNotify_ClosedEntered() {is=title K2Node_Event}
   > Set DeployAnimAlpha(DeployAnimAlpha=1)
   > Set bCanFlap(bCanFlap=false)
   > Set bCanFlapTransition(bCanFlapTransition=false)
   > Set bHideBackpackOwl(bHideBackpackOwl=true)
   > Set Actor Hidden In Game(self=@n21, bNewHidden=true)
   n23 = GetOwningComponent()
   HideBoneByName!(execute, self:Skinned Mesh Component Object Reference=@n23, BoneName:Name=Glider_Base, PhysBodyOption:EPhysBodyOp Enum=PBO_None) -> (then)
   UnHideBoneByName!(execute, self:Skinned Mesh Component Object Reference=@n23, BoneName:Name=Glider_Base) -> (then)
  ]}
]
variables [
 BlockAnimRule {type=boolean}
 Flap_Delay {type=number}
 Left_Foot_loc {type=Vector}
 Left_Foot_Rot {type=Rotator}
 Right_Foot_Loc {type=Vector}
 Right_Foot_Rot {type=Rotator}
]
