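# Material function for Build-A-Brella glider customization that recolors base color with user-set main and secondary color parameters.
# Each color is remapped to a limited range, then expanded into dark and light variants blended by the incoming base color's red channel, preserving shading.
# The BAB_FX mask applies main color through red, then secondary through green.
# The "Use Glider_BAB" static switch outputs the recolored attributes or passes the input through unchanged.
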MaterialFunction /Game/Gadgets/Assets/VinderTech_GliderChute/Materials/Functions/MF_Glider_BAB
graph {
 properties {description=Build-A-Brella}
 nodes [
  Constant() {r=0.6}
  Constant() {r=0.6}
  n1 = FunctionInput() {inputType=FunctionInput_MaterialAttributes}
  n2 = GetMaterialAttributes(None=@n1) {attributeGetTypes [#1]}
  n3 = Constant() {r=0.015}
  n4 = Constant() {r=0.8}
  n5 = VectorParameter() {group=BAB, parameterName=BAB_MainColor, defaultValue=<rgba 0.2623 0.2623 0.2623 1>}
  n6 = MaterialFunctionCall(
   X=LinearInterpolate(A=@n3, B=@n4, Alpha=@n5.R)
   Y=LinearInterpolate(A=@n3, B=@n4, Alpha=@n5.G)
   Z=LinearInterpolate(A=@n3, B=@n4, Alpha=@n5.B)
  ) {materialFunction=&MakeFloat3}
  n7 = Saturate(None=Multiply(A=@n6.Result, B=Constant() {r=0.25}))
  n8 = Clamp(None=Multiply(A=Constant() {r=1.25}, B=@n6.Result), Min=Constant(), Max=@n4) {minDefault=0.015}
  n9 = ComponentMask(None=@n2.BaseColor) {r=true}
  n10 = TextureSampleParameter2D() {group=BAB, parameterName=BAB_FX, samplerType=SAMPLERTYPE_Masks, texture=&T_BlackMask}
  n11 = Constant() {r=0.015}
  n12 = Constant() {r=0.8}
  n13 = VectorParameter() {group=BAB, parameterName=BAB_SecondColor, defaultValue=<rgba 0.7835 0.7835 0.7835 1>}
  n14 = MaterialFunctionCall(
   X=LinearInterpolate(A=@n11, B=@n12, Alpha=@n13.R)
   Y=LinearInterpolate(A=@n11, B=@n12, Alpha=@n13.G)
   Z=LinearInterpolate(A=@n11, B=@n12, Alpha=@n13.B)
  ) {materialFunction=&MakeFloat3}
  n15 = Multiply(A=@n14.Result, B=Constant() {r=0.25})
  n16 = Multiply(A=Constant() {r=1.25}, B=@n14.Result)
  n17 = Constant()
  n18 = LinearInterpolate(
   A=LinearInterpolate(A=@n2.BaseColor, B=LinearInterpolate(A=@n7, B=@n8, Alpha=@n9), Alpha=@n10.R)
   B=LinearInterpolate(A=Saturate(None=@n15), B=Clamp(None=@n16, Min=@n17, Max=@n12) {minDefault=0.015}, Alpha=@n9)
   Alpha=@n10.G
  )
  FunctionOutput(None=StaticSwitchParameter(
    True=SetMaterialAttributes!(MaterialAttributes=@n2.MaterialAttributes, BaseColor=@n18) {attributeSetTypes [#1]}
    False=@n1
   ) {parameterName=Use Glider_BAB})
 ]
}
