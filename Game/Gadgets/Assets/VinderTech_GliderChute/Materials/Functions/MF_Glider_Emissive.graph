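# MF_Glider_Emissive is a material function that sets EmissiveColor on incoming material attributes for glider materials.
# It crops EmissiveTexture with a UV-corner parameter, whose B channel feeds both lower-right components, then multiplies by EmissiveMultiplier.
# An inverted Fresnel fades the glow toward grazing angles, skipped at Low quality.
# The static switch "Use TOD Emissive Boost" optionally routes the result through BoostEmissivebyTOD.
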MaterialFunction /Game/Gadgets/Assets/VinderTech_GliderChute/Materials/Functions/MF_Glider_Emissive
graph {nodes [
  n1 = GetMaterialAttributes(None=FunctionInput() {inputName=Material, inputType=FunctionInput_MaterialAttributes})
  n2 = VectorParameter() {group=Emissive, parameterName=Emissive Texture UVs RG_TopLeft BA_BottomRight, defaultValue=<rgba 0 0 1 1>}
  n3 = Multiply(
   A=MaterialFunctionCall!(
    UVs
    TextureIn=TextureObjectParameter() {group=Emissive, parameterName=EmissiveTexture, texture=&T_BlacksRGB}
    UpperLeftCorner=ComponentMask(None=@n2.RGB) {g=true, r=true}
    LowerRightCorner=AppendVector(A=@n2.B, B=@n2.B)
    ExportFloat4
   ) {materialFunction=/Engine/Functions/Engine_MaterialFunctions01/Texturing/TextureCropping}.CroppedMasked
   B=VectorParameter() {group=Emissive, parameterName=EmissiveMultiplier, defaultValue=<rgba 1 1 1 1>}.RGB
  )
  n4 = Fresnel(ExponentIn=ScalarParameter() {defaultValue=1, parameterName=Emissive Fresnel Exponent}) {baseReflectFraction=0}
  n5 = Multiply(A=@n3, B=Saturate(None=LinearInterpolate(Alpha=@n4) {constA=1, constB=0}))
  n6 = QualitySwitch(Default=@n5, Low=@n3, High=@n5, Medium=@n5)
  n7 = MaterialFunctionCall!(In=@n6) {materialFunction=/Game/Weapons/WeaponSkins/MaterialFunction/BoostEmissivebyTOD}
  FunctionOutput(None=StaticSwitchParameter(
    True=SetMaterialAttributes!(MaterialAttributes=@n1.MaterialAttributes, EmissiveColor=@n7.Result) {attributeSetTypes [#1]}
    False=SetMaterialAttributes(MaterialAttributes=@n1.MaterialAttributes, EmissiveColor=@n6) {attributeSetTypes [#1]}
   ) {group=Skin FX, parameterName=Use TOD Emissive Boost})
 ]}
