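# Material function that supplies the shared texture inputs for glider materials as a single material attributes output.
# Diffuse and Normals texture parameters feed base color and normal directly.
# A packed SpecularMasks texture drives specular from red and metallic from green.
# Its blue channel remaps roughness between the SpecRoughnessMin and SpecRoughnessMax scalar parameters, letting materials tune the range.
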MaterialFunction /Game/Gadgets/Assets/VinderTech_GliderChute/Materials/Functions/MF_Glider_TextureInputs
graph {nodes [
  n1 = TextureSampleParameter2D() {parameterName=SpecularMasks, samplerType=SAMPLERTYPE_Masks, texture=&T_BlackMask}
  n2 = ScalarParameter() {group=Roughness, parameterName=SpecRoughnessMin}
  n3 = ScalarParameter() {defaultValue=1, group=Roughness, parameterName=SpecRoughnessMax}
  FunctionOutput(None=MakeMaterialAttributes!(
    BaseColor=TextureSampleParameter2D() {parameterName=Diffuse, texture=&T_BlacksRGB}.RGB
    Metallic=@n1.G
    Specular=@n1.R
    Roughness=LinearInterpolate(A=@n2, B=@n3, Alpha=@n1.B)
    Anisotropy
    EmissiveColor
    Opacity
    OpacityMask
    Normal=TextureSampleParameter2D() {parameterName=Normals, samplerType=SAMPLERTYPE_Normal, texture=&FlatNormalBlue}.RGB
    Tangent
    WorldPositionOffset
    SubsurfaceColor
    ClearCoat
    ClearCoatRoughness
    AmbientOcclusion
    Refraction
    CustomizedUVs_0
    CustomizedUVs_1
    CustomizedUVs_2
    CustomizedUVs_3
    CustomizedUVs_4
    CustomizedUVs_5
    CustomizedUVs_6
    CustomizedUVs_7
    PixelDepthOffset
    ShadingModel
    Displacement
   ))
 ]}
