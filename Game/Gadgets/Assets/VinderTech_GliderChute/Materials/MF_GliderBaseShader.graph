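# Base shader material function for glider gadgets, outputting full material attributes from diffuse, packed specular mask and normal textures.
# Alpha and res-in effect drive the opacity mask, with switches for clear-coat metallic and vertex-color hiding via WPO.
# Cropped emissive gains supercharge glow and a back-face-only Fresnel gradient, both dropped on mobile and low/medium quality.
# WeaponLocalPosition feeds customized UVs 2 and 3.
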MaterialFunction /Game/Gadgets/Assets/VinderTech_GliderChute/Materials/MF_GliderBaseShader
graph {nodes [
  n1 = FunctionInput(Preview=TextureCoordinate()) {bUsePreviewValueAsDefault=true, inputName=UV 1 coordinates, inputType=FunctionInput_Vector2, sortPriority=2}
  n2 = TextureSampleParameter2D(UVs=@n1) {
   group=Diffuse
   parameterName=Diffuse
   samplerSource=SSM_Wrap_WorldGroupSettings
   texture=/Game/Weapons/FORT_Shotguns/Textures/T_HammerShotGun_D
  }
  n3 = TextureSampleParameter2D(UVs=@n1) {
   desc="R = SpecMask G = Metallic B = Roughness"
   group=Specular
   parameterName=SpecularMasks
   samplerSource=SSM_Wrap_WorldGroupSettings
   samplerType=SAMPLERTYPE_Masks
   texture=/Game/Weapons/FORT_Shotguns/Textures/T_HammerShotGun_S
  }
  n4 = LinearInterpolate(A=@n3.G, B=Multiply(A=@n3.A, B=ScalarParameter() {defaultValue=1, group=ClearCoat, parameterName=PaintMetallic}), Alpha=@n3.G)
  n5 = ScalarParameter() {defaultValue=1, group=Specular, parameterName=SpecularLevel}
  n6 = ScalarParameter() {group=Roughness, parameterName=SpecRoughnessMin}
  n7 = ScalarParameter() {defaultValue=1, group=Roughness, parameterName=SpecRoughnessMax}
  n8 = MaterialFunctionCall!(Animation Gradient, Rez In Percentage) {materialFunction=/Game/Packages/Fortress_SharedMaterials/Base_Material_Functions/CharacterResInEffect}
  n9 = TextureSampleParameter2D(UVs=@n1) {group=Normals, parameterName=Normals, samplerType=SAMPLERTYPE_Normal, texture=/Game/Weapons/FORT_Shotguns/Textures/T_HammerShotGun_N}
  n10 = MaterialFunctionCall!(WPO) {materialFunction=/Game/Packages/Fortress_SharedMaterials/Base_Material_Functions/Fortnite_HideElementsWithVertexColor}
  n11 = Constant()
  n12 = VectorParameter() {group=Emissive, parameterName=Emissive Texture UVs RG_TopLeft BA_BottomRight, defaultValue=<rgba 0 0 1 1>}
  n13 = Multiply(
   A=MaterialFunctionCall!(
    UVs
    TextureIn=TextureObjectParameter() {group=Emissive, parameterName=EmissiveTexture, texture=&T_BlacksRGB}
    UpperLeftCorner=ComponentMask(None=@n12.RGB) {g=true, r=true}
    LowerRightCorner=AppendVector(A=@n12.B, B=@n12.B)
    ExportFloat4
   ) {materialFunction=/Engine/Functions/Engine_MaterialFunctions01/Texturing/TextureCropping}.CroppedMasked
   B=VectorParameter() {group=Emissive, parameterName=EmissiveMultiplier, defaultValue=<rgba 1 1 1 1>}.RGB
  )
  n14 = MaterialFunctionCall!(
   AverageHandPosition
   SuperChargedFlickerBias=ScalarParameter() {defaultValue=1, group=Super Charge, parameterName=SuperChargedFlickerBias}
   SuperChargedEffect=ScalarParameter() {group=Super Charge, parameterName=SuperChargedEffect}
   UseDistanceCalc=StaticBool()
   Emissive
  ) {materialFunction=/Game/Characters/Fortress_CharacterShader/SuperChargedGlow}
  n15 = Clamp(None=Multiply(A=TwoSidedSign()) {constB=-1})
  n16 = VectorParameter() {parameterName=colcol1, defaultValue=<rgba 6.5583 9.5135 20 0.2>}
  n17 = StaticBool()
  n18 = LinearInterpolate(
   A=@n16.RGB
   B=VectorParameter() {parameterName=colcol2, defaultValue=<rgba 1.1035 1.5954 6 1>}.RGB
   Alpha=Abs(None=MaterialFunctionCall(Power=@n16.A, Clamp Fresnel Dot Product=@n17) {materialFunction=&Fresnel_Function}.Result)
  )
  n19 = ShadingPathSwitch(Default=Add(A=Add(A=@n13, B=@n14.Result), B=Multiply(A=@n15, B=@n18)), Mobile=@n13) {desc=Migrated from FeatureLevelSwitch}
  n20 = MaterialFunctionCall() {materialFunction=/Game/Packages/Fortnite_UniqueMaterials_1/Weapons/WeaponLocalPosition}
  FunctionOutput(None=MakeMaterialAttributes!(
    BaseColor=@n2.RGB
    Metallic=StaticSwitchParameter(True=@n4, False=@n3.G) {parameterName=isClearCoat}
    Specular=Multiply(A=@n3.R, B=@n5)
    Roughness=LinearInterpolate(A=@n6, B=@n7, Alpha=@n3.B)
    Anisotropy
    EmissiveColor=QualitySwitch(Default=@n19, Low=@n13, High=@n19, Medium=@n13)
    Opacity=Constant() {r=1}
    OpacityMask=Multiply(A=@n2.A, B=@n8.Modulative Opacity)
    Normal=@n9.RGB
    Tangent
    WorldPositionOffset=StaticSwitchParameter(True=@n10.Result, False=@n11) {group=Opacity Mask, parameterName=Use Vertex Colors for Mask}
    SubsurfaceColor=Constant3Vector()
    ClearCoat=@n3.A
    ClearCoatRoughness=ScalarParameter() {group=ClearCoat, parameterName=ClearCoatRoughness}
    AmbientOcclusion
    Refraction
    CustomizedUVs_0
    CustomizedUVs_1
    CustomizedUVs_2=@n20.Custom Tex Coord 2
    CustomizedUVs_3=@n20.Custom Tex Coord 3
    CustomizedUVs_4
    CustomizedUVs_5
    CustomizedUVs_6
    CustomizedUVs_7
    PixelDepthOffset
    ShadingModel
    Displacement
   )) {outputName=Result Material Attributes}
 ]}
